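; Boss NPC pawn for the WildEstate SilentNovel_VaneHush character, switching between a prop and its weapon by bot alert level.
; On BeginPlay it links its animation layers, spawns a cosmetic portal effect, calls ForceHideHealthBar and binds to the AI controller's alert changes.
; The alert handler toggles PropVisibility, whose RepNotify shows or hides the maraca, and sets weapon holstering, with Threatened hiding the prop.
; EndPlay unlinks the layers and unbinds the event.
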(event Custom|SpawnRiftEffect
  (Networking|CanExecuteCosmeticEvents
    (:True
      (Niagara|SpawnSystematLocation "/WildEstateBosses/Gameplay/RiftEffect/NS_WildEstate_Portal_EntryExit.NS_WildEstate_Portal_EntryExit" (Transformation|GetActorLocation)))))

(event EventBeginPlay
  (|CallParentFunction)
  (Components|SkeletalMesh|AnimationBlueprintLinking|LinkAnimClassLayers (Variables|Character|GetMesh) "/SilentNovel_VaneHush/Assets/Animation/WildEstate_NPC_SilentNovel_VaneHush_AnimBP.WildEstate_NPC_SilentNovel_VaneHush_AnimBP_C")
  (CallFunction|SpawnRiftEffect)
  (CallFunction|ForceHideHealthBar 0.0)
  (bind _asfort_athena_aibot_controller (Utilities|Casting|CastToFortAthenaAIBotController (AI|GetAIController self)))
  (AlertLevel|BindEventtoOnAlertLevelChangedEvent _asfort_athena_aibot_controller (delegate Custom|OnAlertLevelChanged_WildEstate))
  (CallFunction|OnAlertLevelChangedWildEstate _asfort_athena_aibot_controller))

(event Custom|OnAlertLevelChanged_WildEstate
  (params (BotController :type Fort%20Athena%20AIBot%20Controller%20Object%20Reference) (OldAlertLevel :type EAlertLevel%20Enum :default "Unaware") (NewAlertLevel :type EAlertLevel%20Enum :default "Unaware"))
  (switch Utilities|FlowControl|Switch|SwitchonEAlertLevel NewAlertLevel
    (:Unaware
      (Variables|Default|SetPropVisibility true))
    (:Suspicious
      (Variables|Default|SetPropVisibility true :!label _label))
    (:LostTarget
      (goto _label))
    (:Threatened
      (Variables|Default|SetPropVisibility false)
      (Weapon|SetHolsterWeapon 0 false true)))
  (Weapon|SetHolsterWeapon 0 true))

(event Collision|EventActorBeginOverlap
  (params (OtherActor :type Actor%20Object%20Reference))
  (|CallParentFunction OtherActor) ; This node is disabled and will not be called.\nDrag off pins to build functionality.
)

(event EventTick
  (params (DeltaSeconds :type Float%20%28single-precision%29))
  (|CallParentFunction DeltaSeconds) ; This node is disabled and will not be called.\nDrag off pins to build functionality.
)

(event EventEndPlay
  (params (EndPlayReason :type EEndPlayReason%20Enum :default "Destroyed"))
  (Components|SkeletalMesh|AnimationBlueprintLinking|UnlinkAnimClassLayers (Variables|Character|GetMesh) "/SilentNovel_VaneHush/Assets/Animation/WildEstate_NPC_SilentNovel_VaneHush_AnimBP.WildEstate_NPC_SilentNovel_VaneHush_AnimBP_C")
  (bind _asfort_athena_aibot_controller (Utilities|Casting|CastToFortAthenaAIBotController (AI|GetAIController self)))
  (AlertLevel|UnbindEventfromOnAlertLevelChangedEvent _asfort_athena_aibot_controller (EventDispatchers|CreateEvent :!event Custom|OnAlertLevelChanged_WildEstate)))

(fn OnRep_PropVisibility
  (Rendering|SetVisibility (Variables|Default|GetMaraca_L) (Variables|Default|GetPropVisibility) true))

(fn ConstructionScript
  (|CallParentFunction))
